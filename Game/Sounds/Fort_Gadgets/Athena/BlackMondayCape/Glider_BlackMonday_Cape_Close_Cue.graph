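# Sound cue for the close sound of the BlackMonday cape glider in Athena.
# It plays a single close wave through a modulator that slightly randomizes volume, with the whole cue pitched down.
# It uses the shared glider open/close attenuation and concurrency settings and routes to the glider landing sound class.
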SoundCue /Game/Sounds/Fort_Gadgets/Athena/BlackMondayCape/Glider_BlackMonday_Cape_Close_Cue
root=Modulator_0
nodes [
 {id=Modulator_0, children [WavePlayer_1], properties {volumeMax=0.85, volumeMin=0.75}}
 {id=WavePlayer_1, properties {soundWaveAssetPtr=&/Glider_BlackMonday_Cape_Close_01}}
]
properties {
 attenuationSettings=/Game/Sounds/Attenuation/Fort_Player_Glider_OpenClose_Athena_Att
 maxDistance=10000
 pitchMultiplier=0.8
 soundClassObject=&Fort_Gliders_Land
 concurrencySet [&Athena_GliderOpenClose_Conc]
}
