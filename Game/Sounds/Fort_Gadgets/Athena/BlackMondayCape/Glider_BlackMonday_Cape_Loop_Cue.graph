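# Looping Sound Cue for the BlackMonday cape glider in Fortnite Athena.
# A mixer layers two looping cape waves, Loop_01 and Loop_02, with the second mixed slightly quieter.
# The cue lowers overall pitch and volume, uses the player glider thrust attenuation, and plays in the glider land sound class.
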SoundCue /Game/Sounds/Fort_Gadgets/Athena/BlackMondayCape/Glider_BlackMonday_Cape_Loop_Cue
root=Mixer_0
nodes [
 {id=Mixer_0, children [WavePlayer_3,WavePlayer_2], properties {inputVolume [1,0.7]}}
 {
  id=WavePlayer_2
  properties {bLooping=true, soundWaveAssetPtr=&/Glider_BlackMonday_Cape_Loop_02}
 }
 {
  id=WavePlayer_3
  properties {bLooping=true, soundWaveAssetPtr=&/Glider_BlackMonday_Cape_Loop_01}
 }
]
properties {
 attenuationSettings=&Fort_Player_Glider_Thrust_Athena_Att
 maxDistance=12500
 pitchMultiplier=0.8
 soundClassObject=&Fort_Gliders_Land
 volumeMultiplier=0.3
}
