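# Sound cue for the grapple-open sound of the BlackMonday Cape glider in Athena.
# A short fixed delay feeds a single wave player for the OpenGrapple sound.
# It plays at reduced volume through the glider open/close attenuation, the Fort_Gliders_Land sound class and a glider open/close concurrency set.
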SoundCue /Game/Sounds/Fort_Gadgets/Athena/BlackMondayCape/Glider_BlackMonday_Cape_OpenGrapple_Cue
root=Delay_1
nodes [
 {id=Delay_1, children [WavePlayer_1], properties {delayMax=0.07, delayMin=0.07}}
 {id=WavePlayer_1, properties {soundWaveAssetPtr=&/Glider_BlackMonday_Cape_OpenGrapple_01}}
]
properties {
 attenuationSettings=/Game/Sounds/Attenuation/Fort_Player_Glider_OpenClose_Athena_Att
 maxDistance=10000
 soundClassObject=&Fort_Gliders_Land
 volumeMultiplier=0.5
 concurrencySet [&Athena_GliderOpenClose_Conc]
}
