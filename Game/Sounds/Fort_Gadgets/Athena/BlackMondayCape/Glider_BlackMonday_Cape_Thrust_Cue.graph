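# Sound cue for the thrust sound of the BlackMonday cape glider.
# A short fixed delay feeds a modulator, then a random node picks one of three equally weighted Grape body-swing waves borrowed from BlackMonday weapon audio.
# The cue plays quieter and slightly pitched down, using glider open/close attenuation, the glider landing sound class and a thrust concurrency group.
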SoundCue /Game/Sounds/Fort_Gadgets/Athena/BlackMondayCape/Glider_BlackMonday_Cape_Thrust_Cue
root=Delay_0
nodes [
 {id=Delay_0, children [Modulator_1], properties {delayMax=0.35, delayMin=0.35}}
 {id=Modulator_1, children [Random_1]}
 {id=Random_1, children [WavePlayer_2,WavePlayer_3,WavePlayer_4], properties {weights [1,1,1]}}
 {id=WavePlayer_2, properties {soundWaveAssetPtr=/Game/Sounds/Fort_Weapons/BlackMonday/Grape/BM_Grape_BodySwing_01}}
 {id=WavePlayer_3, properties {soundWaveAssetPtr=/Game/Sounds/Fort_Weapons/BlackMonday/Grape/BM_Grape_BodySwing_03}}
 {id=WavePlayer_4, properties {soundWaveAssetPtr=/Game/Sounds/Fort_Weapons/BlackMonday/Grape/BM_Grape_BodySwing_02}}
]
properties {
 attenuationSettings=/Game/Sounds/Attenuation/Fort_Player_Glider_OpenClose_Athena_Att
 maxDistance=10000
 pitchMultiplier=0.9
 soundClassObject=&Fort_Gliders_Land
 volumeMultiplier=0.4
 concurrencySet [&Athena_Glider_Thrust_Conc]
}
